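; Gameplay volume actor for Spires that initializes spawned hero spires and tracks the pawns inside it.
; On BeginPlay it binds to the variant spawn location's OnActorSpawned, then calls InitializeHeroSpire on spawned hero spire wrappers, passing itself as reference.
; Overlap begin and end events add and remove FortPawns in PawnsInVolume, guarded by membership checks.
; Each membership change broadcasts OnPawnEnterVolume or OnPawnLeaveVolume.
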(event EventBeginPlay
  (EventDispatchers|AssignOnActorSpawned (Variables|ForPlayspace|GetVariantSpawnLocation) (delegate Custom|OnActorSpawned_Event)))

(event Custom|OnActorSpawned_Event
  (params (SpawnedActor :type Actor%20Object%20Reference))
  (bind _asb_hero_spire_wrapper_parent (Utilities|Casting|CastToB_HeroSpire_Wrapper_Parent SpawnedActor))
  (Class|BHeroSpireWrapperParent|InitializeHeroSpire _asb_hero_spire_wrapper_parent self))

(event OnComponentBeginOverlap(Overlap)
  (params (OverlappedComponent :type Primitive%20Component%20Object%20Reference) (OtherActor :type Actor%20Object%20Reference) (OtherComp :type Primitive%20Component%20Object%20Reference) (OtherBodyIndex :type Integer) (bFromSweep :type Boolean) (SweepResult :type Hit%20Result%20Structure%20%28by%20ref%29))
  ;[0 0.05421 0.875] Handle pawns entering and exiting the volume
  (bind _asfort_pawn (Utilities|Casting|CastToFortPawn OtherActor))
  (if (not (Utilities|Array|ContainsItem (Variables|Default|GetPawnsInVolume) _asfort_pawn))
    (Utilities|Array|AddUnique (Variables|Default|GetPawnsInVolume) _asfort_pawn)
    (Default|CallOnPawnEnterVolume _asfort_pawn)))

(event OnComponentEndOverlap(Overlap)
  (params (OverlappedComponent :type Primitive%20Component%20Object%20Reference) (OtherActor :type Actor%20Object%20Reference) (OtherComp :type Primitive%20Component%20Object%20Reference) (OtherBodyIndex :type Integer))
  ;[0 0.05421 0.875] Handle pawns entering and exiting the volume
  (bind _asfort_pawn (Utilities|Casting|CastToFortPawn OtherActor))
  (if (Utilities|Array|ContainsItem (Variables|Default|GetPawnsInVolume) _asfort_pawn)
    (Utilities|Array|RemoveItem (Variables|Default|GetPawnsInVolume) _asfort_pawn)
    (Default|CallOnPawnLeaveVolume _asfort_pawn)))

(fn ConstructionScript)
