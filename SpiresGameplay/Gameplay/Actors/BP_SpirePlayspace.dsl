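; Playspace Blueprint for a Spire that ties the lifetime of its bound gameplay volume to that Spire.
; On BeginPlay it listens for the gameplay volume being assigned, then gets the playspace's Spire and binds to its death event.
; When the Spire dies, it puts a short lifespan on the bound volume so it is destroyed shortly after.
; GetSpireFromGameplayVolume casts the volume to BP_SpireGameplayVolume and caches and returns its Spire.
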(event EventBeginPlay
  (Playspace|BindEventtoNotifyBoundGameplayVolumeSet (EventDispatchers|CreateEvent :!event Custom|OnVolumeAssigned)))

(event Custom|OnVolumeAssigned
  (params (Playspace :type Playspace%20Object%20Reference))
  ; Spire Death Handling.
  (bind _asspire_playspace (Utilities|Casting|CastToSpirePlayspace Playspace))
  (bind _returnvalue (CallFunction|GetSpire _asspire_playspace))
  (Utilities|IsValid _returnvalue
    (:Is%20Valid
      (EventDispatchers|BindEventtoOnDied _returnvalue (delegate Custom|OnSpireDied)))))

(event Custom|OnSpireDied
  (params (DamagedActor :type Actor%20Object%20Reference) (Damage :type Float%20%28single-precision%29) (InstigatedBy :type Controller%20Object%20Reference) (DamageCauser :type Actor%20Object%20Reference) (HitLocation :type Vector) (FHitComponent :type Primitive%20Component%20Object%20Reference) (BoneName :type Name) (Momentum :type Vector))
  ; Spire Death Handling.
  (Actor|SetLifeSpan (Variables|Playspace|GetBoundGameplayVolume) 3.0) ; Destroy Playspace with the spire
)

(fn GetSpireFromGameplayVolume
  (bind _asbp_spire_gameplay_volume (Utilities|Casting|CastToBP_SpireGameplayVolume (Variables|Playspace|GetBoundGameplayVolume))
    (:then
      (bind _output_get (Variables|Default|SetSpire (Variables|/SpiresGameplay/Gameplay/Actors/BP_SpireGameplayVolume.BP_SpireGameplayVolume_C|GetSpire _asbp_spire_gameplay_volume)))
      (return _output_get))
    (:CastFailed
      (return 0))))

(fn ConstructionScript)
