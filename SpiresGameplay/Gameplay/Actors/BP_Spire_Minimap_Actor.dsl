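; BP_Spire_Minimap_Actor is an actor that carries a Fort minimap entry for Spires gameplay, exposing functions to set its icon and map legend text.
; SetMapIcon and SetMapLegendText each write to the actor's FortMiniMap data, but only when running with authority.
; Both flush net dormancy after the write so the change replicates from the otherwise dormant actor.
; BeginPlay, Tick, overlap and the construction script contain no logic.
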(event EventBeginPlay)

(event Collision|EventActorBeginOverlap
  (params (OtherActor :type Actor%20Object%20Reference)))

(event EventTick
  (params (DeltaSeconds :type Float%20%28single-precision%29)))

(fn SetMapIcon
  (params (MiniMapIcon :type Object%20Reference))
  (Networking|SwitchHasAuthority
    (:Authority
      (MiniMapData|SetMiniMapIcon (Variables|Default|GetFortMiniMap) MiniMapIcon)
      (Networking|FlushNetDormancy))))

(fn SetMapLegendText
  (params (Icon%20Text :type Text))
  (Networking|SwitchHasAuthority
    (:Authority
      (MiniMapData|SetMiniMapIconMapLegendText (Variables|Default|GetFortMiniMap) Icon%20Text)
      (Networking|FlushNetDormancy))))

(fn ConstructionScript)
