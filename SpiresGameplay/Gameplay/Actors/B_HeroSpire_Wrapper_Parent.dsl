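; Parent wrapper actor for Hero Spires, providing a base for child classes to implement a spire ability.
; On initialization it stores its gameplay volume and listens on the event router for Hero Spire commands, sent by the Major Spire, that activate or deactivate the ability.
; It also binds pawn enter/leave callbacks, left empty here like the ability hooks that only set an active flag.
; End play stops listening and optionally deactivates the ability.
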(event Custom|InitializeHeroSpire
  (params (GameplayVolume :type BP%20Spire%20Gameplay%20Volume%20Object%20Reference))
  (bind _gameplayvolume (Variables|Default|GetGameplayVolume))
  ;[0 0.05421 0.875] Wait for the Volume to init the Hero Spire
  (Variables|Default|SetGameplayVolume GameplayVolume)
  ;[0 0.05421 0.875] By default, Hero Spire listens for event to be sent by Major Spire to invoke its ability
  (bind _returnvalue (GameplayEventRouter|GetEventRouter "/SpiresGameplay/Gameplay/Actors/BP_SpirePlayspace.BP_SpirePlayspace_C" self))
  (bind (_asynctaskproxy _payload) (GameplayEventRouter|StartListeningToEvent _returnvalue "/SpiresGameplay/Gameplay/Actors/SHeroSpireCommand.SHeroSpireCommand")
    (:then
      (Variables|Default|SetWaitEventTask _asynctaskproxy)
      ;[0 0.05421 0.875] Set up bindings for handling pawns entering/leaving if necessary
      (Default|BindEventtoOnPawnEnterVolume _gameplayvolume (EventDispatchers|CreateEvent :!event OnNewPawnEnter))
      (Default|BindEventtoOnPawnLeaveVolume _gameplayvolume (EventDispatchers|CreateEvent :!event OnPawnLeave)))
    (:OnEventReceived
      (bind _payload_heroabilityenabled__2_3e89373045af8f212530288f452c44a9 (Utilities|Struct|BreakSHeroSpireCommand _payload))
      (if _payload_heroabilityenabled__2_3e89373045af8f212530288f452c44a9
        (CallFunction|ActivateHeroSpireAbility)
        (else
          (CallFunction|DeactivateHeroSpireAbility))))))

(event Custom|ActivateHeroSpireAbility
  ;[0 0.05421 0.875] Override in child classes
  (Variables|Default|SetHeroSpireAbilityActive? true))

(event EventEndPlay
  (params (EndPlayReason :type EEndPlayReason%20Enum :default "Destroyed"))
  ;[1 0 0.04924] End Play
  (GameplayEventRouter|StopListeningtoEvent (Variables|Default|GetWaitEventTask))
  (if (Variables|HeroSpire|GetDeactivateAbilityOnDestroyed?)
    (CallFunction|DeactivateHeroSpireAbility)))

(event Custom|DeactivateHeroSpireAbility
  ;[0 0.05421 0.875] Override in child classes
  (Variables|Default|SetHeroSpireAbilityActive? false))

(fn OnNewPawnEnter
  (params (Pawn :type Fort%20Pawn%20Object%20Reference)))

(fn OnPawnLeave
  (params (Pawn :type Fort%20Pawn%20Object%20Reference)))

(fn ConstructionScript)
