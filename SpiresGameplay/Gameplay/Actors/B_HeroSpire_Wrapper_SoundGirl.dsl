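; B_HeroSpire_Wrapper_SoundGirl is the SoundGirl variant of a hero spire wrapper actor in Spires gameplay, and it adds no logic of its own.
; Its activate and deactivate hero spire ability events and its construction script only call the parent implementation.
; The pawn enter and leave events are empty stubs, so any SoundGirl-specific behaviour is not visible in this graph.
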(event EventActivateHeroSpireAbility
  (|CallParentFunction))

(event EventDeactivateHeroSpireAbility
  (|CallParentFunction))

(event EventOnNewPawnEnter
  (params (Pawn :type Fort%20Pawn%20Object%20Reference)))

(event EventOnPawnLeave
  (params (Pawn :type Fort%20Pawn%20Object%20Reference)))

(fn ConstructionScript
  (|CallParentFunction))
